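# Animation Blueprint for the SouthBalk backpack that moves its cube bone with player motion and corrects the pack's position as the character's spine bends.
# The spine_05-to-spine_01 yaw delta on a character mesh part offsets and rolls base_cube.
# Cached poses feed that correction into stacked additive dyn_cube offsets, weighted by interpolated movement values from a deprecated helper.
# Sine-wave vectors are computed, but the final dyn_cube and dyn_sphere nodes show zero translation.
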AnimBlueprint /BRCosmetics/Accessories/FORT_Backpacks/Backpack_M_MED_SouthBalk/Meshes/M_MED_SouthBalk_Pack_AnimBP
target_skeleton=&/M_MED_SouthBalk_Pack_Skeleton
graphs [
 AnimGraph {nodes [
   n1 = "Transform (Modify) Bone - Bone: dyn_cube"!(ComponentPose=Local To Component(LocalPose=Use cached pose 'CRCache_Drag'()), Alpha=1, Translation="0,0,0") -> (Pose) {translationMode=BMM_Additive, translationSpace=BCS_BoneSpace, boneToModify {boneName=dyn_cube}}
   Output Pose(Result=Component To Local(ComponentPose="Transform (Modify) Bone - Bone: dyn_sphere"!(ComponentPose=@n1, Alpha=1, Translation="0,0,0") -> (Pose) {translationMode=BMM_Additive, translationSpace=BCS_BoneSpace, boneToModify {boneName=dyn_sphere}})) {layerGroup=None}
   n2 = "Transform (Modify) Bone - Bone: base_cube"!(
    ComponentPose=Local To Component(LocalPose=Input Pose())
    bAlphaBoolEnabled:bool=True
    Translation=Get Backpack Translation() -> (Backpack Translation:Vector)
    Rotation=Get Backpack Rotate() -> (Backpack Rotate:Rotator)
    Scale="1,1,1"
   ) -> (Pose) {
    alphaInputType=Bool
    rotationMode=BMM_Additive
    translationMode=BMM_Additive
    translationSpace=BCS_BoneSpace
    boneToModify {boneName=base_cube}
   }
   Save cached pose 'CRCache_Correction'(Pose=Component To Local(ComponentPose=@n2)) {cacheName=CRCache_Correction}
   n3 = "Transform (Modify) Bone - Bone: dyn_cube"!(
    ComponentPose=Local To Component(LocalPose=Use cached pose 'CRCache_Correction'())
    Alpha=Get yMovement() -> (yMovement:double)
    Translation="0,-7,0"
    Rotation:Rotator="0,0,-20"
   ) -> (Pose) {
    rotationMode=BMM_Additive
    translationMode=BMM_Additive
    translationSpace=BCS_BoneSpace
    alphaScaleBiasClamp {bInterpResult=true, interpSpeedDecreasing=5, interpSpeedIncreasing=2}
    boneToModify {boneName=dyn_cube}
    rotation {roll=-20}
    translation {y=-7}
   }
   n4 = "Transform (Modify) Bone - Bone: dyn_cube"!(ComponentPose=@n3, Alpha=Get xMovementPos() -> (xMovementPos:double), Translation="0,0,-5") -> (Pose) {
    rotationMode=BMM_Additive
    translationMode=BMM_Additive
    translationSpace=BCS_BoneSpace
    alphaScaleBiasClamp {bInterpResult=true, interpSpeedDecreasing=5, interpSpeedIncreasing=5}
    boneToModify {boneName=dyn_cube}
    translation {z=-5}
   }
   n5 = "Transform (Modify) Bone - Bone: dyn_cube"(ComponentPose=@n4, Alpha=Get xMovementNeg() -> (xMovementNeg:double), Translation="0,0,5") {
    translationMode=BMM_Additive
    translationSpace=BCS_BoneSpace
    alphaScaleBiasClamp {bInterpResult=true, interpSpeedDecreasing=5, interpSpeedIncreasing=5}
    boneToModify {boneName=dyn_cube}
    translation {z=5}
   }
   n6 = "Transform (Modify) Bone - Bone: dyn_cube"(ComponentPose=@n5, Alpha=Get zMovementPos() -> (zMovementPos:double), Translation="-7,0,0") {
    translationMode=BMM_Additive
    translationSpace=BCS_BoneSpace
    alphaScaleBiasClamp {bInterpResult=true, interpSpeedDecreasing=2, interpSpeedIncreasing=2}
    boneToModify {boneName=dyn_cube}
    translation {x=-7}
   }
   n7 = "Transform (Modify) Bone - Bone: dyn_cube"(ComponentPose=@n6, Alpha=Get zMovementNeg() -> (zMovementNeg:double), Translation="7,0,0") {
    translationMode=BMM_Additive
    translationSpace=BCS_BoneSpace
    alphaScaleBiasClamp {bInterpResult=true, interpSpeedDecreasing=2, interpSpeedIncreasing=2}
    boneToModify {boneName=dyn_cube}
    translation {x=7}
   }
   Save cached pose 'CRCache_Drag'(Pose=Component To Local(ComponentPose=@n7)) {cacheName=CRCache_Drag}
  ]}
 EventGraph {nodes [
   Event BlueprintUpdateAnimation()
   n1 = GetRealTimeSeconds()
   n2 = "Sin (Radians)"(A=float + float(A=float * float(A=2, B=@n1)))
   n3 = "Sin (Radians)"(A=float + float(A=float * float(A=0.3, B=@n1), B=-2))
   n4 = "Sin (Radians)"(A=float + float(A=float * float(A=0.3, B=@n1), B=3))
   > Set sinWaveDispCube(sinWaveDispCube:Vector=MakeVector(X=float * float(A=@n2, B=2), Y=float * float(A=@n3, B=0.25), Z=float * float(A=@n4, B=0.75))) -> (Output_Get:Vector)
   n5 = float * float(A=@n1, B=1.5)
   n6 = float * float(A="Sin (Radians)"(A=float + float(A=@n5, B=2)), B=1)
   > Set sinWaveDispSphere(sinWaveDispSphere:Vector=MakeVector(X=float + float(A=@n6, B=1), Y=float * float(A=@n3, B=0.25), Z=float * float(A=@n4, B=0.25))) -> (Output_Get:Vector)
   > n7 = Player Movement Function DEPRECATED!(
    execute
    CharacterPartAnimInstance:Custom Character Part Anim Instance Object Reference=Self-Reference()
    X Dot Product:double=1
    Y Dot Product:double=1
    Z Dot Product:double=1
    DrawDebug:bool=false
    Debug Scalar:double=50
   ) -> (then, X Movement:double, Y Movement:double, Z Movement:double)
   > Set xMovementPos(xMovementPos:double="Clamp (Float)"(Value=@n7, Min=0, Max=1)) -> (Output_Get:double)
   > Set xMovementNeg(xMovementNeg:double=Remap Float(Input=@n7, OldRange=(X=-1,Y=0), NewRange=(X=1,Y=0), Clamp=true) -> (Output=~)) -> (Output_Get:double)
   > Set yMovement(yMovement:double="Clamp (Float)"(Value=@n7.Y Movement, Min=0, Max=1)) -> (Output_Get:double)
   > Set zMovementPos(zMovementPos:double="Clamp (Float)"(Value=@n7.Z Movement, Min=0, Max=1)) -> (Output_Get:double)
   > Set zMovementNeg(zMovementNeg:double=Remap Float(Input=@n7.Z Movement, OldRange=(X=-1,Y=0), NewRange=(X=1,Y=0), Clamp=true) -> (Output=~)) -> (Output_Get:double)
   n8 = Get Skel Mesh()
   n9 = BreakRotator(InRot=BreakTransform(InTransform=GetDeltaTransformFromRefPose(self=@n8, BoneName=spine_05, BaseName=spine_01)).Rotation)
   > Set Backpack Rotate(Backpack Rotate:Rotator=MakeRotator(Roll=MapRangeClamped(Value=@n9.Yaw, InRangeA=0, InRangeB=-20, OutRangeB=20))) -> (Output_Get:Rotator)
   > Set Backpack Translation(Backpack Translation:Vector=MakeVector(Y=MapRangeClamped(Value=@n9.Yaw, InRangeA=0, InRangeB=-20, OutRangeB=-5))) -> (Output_Get:Vector)
   n10 = Event OnCharacterCustomizationLoaded()
   > Set Skel Mesh(Skel Mesh=Get Array Item with EFort Custom Part Type(Array=@n10.CustomMeshParts))
   Reroute Node(InputPin:Skeletal Mesh Component Object Reference=@n10.CustomBaseMesh) -> (OutputPin:Skeletal Mesh Component Object Reference)
  ]}
]
variables [
 Backpack Rotate {type=Rotator}
 Backpack Translation {type=Vector}
 Correction {type=number}
 sinWaveDispCube {type=Vector}
 sinWaveDispSphere {type=Vector}
 Skel Mesh {type=SkeletalMeshComponent}
 xMovementNeg {type=number}
 xMovementPos {type=number}
 yMovement {type=number}
 zMovementNeg {type=number}
 zMovementPos {type=number}
]
